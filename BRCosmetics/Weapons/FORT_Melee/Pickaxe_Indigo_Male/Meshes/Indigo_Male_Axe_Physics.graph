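# Physics asset for the Indigo Male pickaxe, giving capsule bodies and joints to its dyn_ toe and wire bones.
# Three front toes and one back toe each pair with their own MSTR anchor, with a nonzero limit only on swing2.
# The top wire forms a serial chain from its MSTR bone, while the bottom wire fans into four branches with wider swing limits.
# Every joint has a zero twist limit.
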PhysicsAsset /BRCosmetics/Weapons/FORT_Melee/Pickaxe_Indigo_Male/Meshes/Indigo_Male_Axe_Physics
bodies {
 dyn_bot_wireGrp0 [{shapeType=Capsule, length=0.505, radius=0.505, shapeName=dyn_bot_wireGrp0_capsule, rotation=<rot 0 -90 0>}]
 dyn_bot_wireGrp0_0 [{
   shapeType=Capsule
   length=5.8659
   radius=1.5647
   shapeName=dyn_bot_wireGrp0_0_capsule
   center=<3.1047 0.6065 0.1791>
   rotation=<rot 0 92.6409 -89.3458>
  }]
 dyn_bot_wireGrp0_1 [{
   shapeType=Capsule
   length=3.9711
   radius=0.8926
   shapeName=dyn_bot_wireGrp0_1_capsule
   center=<4.462 -1.2505 1.0182>
   rotation=<rot 0 73.4048 -71.7519>
  }]
 dyn_bot_wireGrp0_2 [{
   shapeType=Capsule
   length=4.9246
   radius=1.2689
   shapeName=dyn_bot_wireGrp0_2_capsule
   center=<3.5982 -0.4681 -0.6936>
   rotation=<rot 0 -94.9357 -89.966>
  }]
 dyn_bot_wireGrp0_3 [{
   shapeType=Capsule
   length=5.0898
   radius=0.8546
   shapeName=dyn_bot_wireGrp0_3_capsule
   center=<2.7413 0.2083 0.4662>
   rotation=<rot 0 92.9619 -86.3216>
  }]
 dyn_bot_wireGrp_MSTR [{shapeType=Capsule, length=0.505, radius=0.505, shapeName=dyn_bot_wireGrp_MSTR_capsule, rotation=<rot 0 -90 0>}]
 dyn_c_ba_toe [{
   shapeType=Capsule
   length=8.6355
   radius=4.0882
   shapeName=dyn_c_ba_toe_capsule
   center=<9.0793 0.0025 -0.6016>
   rotation=<rot 0 90.0065 -81.1601>
  }]
 dyn_c_ba_toe_MSTR [{shapeType=Capsule, length=0.505, radius=0.505, shapeName=dyn_c_ba_toe_MSTR_capsule, rotation=<rot 0 -90 0>}]
 dyn_c_fr_toe [{shapeType=Capsule, length=1, radius=9.5, shapeName=dyn_c_fr_toe_capsule, center=<14.5199 -0.1928 2.0761>, rotation=<rot 0 89.3105 -79.6601>}]
 dyn_c_fr_toe_MSTR [{shapeType=Capsule, length=0.505, radius=0.505, shapeName=dyn_c_fr_toe_MSTR_capsule, rotation=<rot 0 -90 0>}]
 dyn_l_fr_toe [{
   shapeType=Capsule
   length=1
   radius=9.5
   shapeName=dyn_l_fr_toe_capsule
   center=<12.8338 -0.5565 3.9967>
   rotation=<rot -0.0001 88.0891 -57.5096>
  }]
 dyn_l_fr_toe_MSTR [{shapeType=Capsule, length=0.505, radius=0.505, shapeName=dyn_l_fr_toe_MSTR_capsule, rotation=<rot 0 -90 0>}]
 dyn_r_fr_toe [{shapeType=Capsule, length=1, radius=9.5, shapeName=dyn_r_fr_toe_capsule, center=<12.7955 0.5351 4.1027>, rotation=<rot 0 91.9054 -57.0936>}]
 dyn_r_fr_toe_MSTR [{shapeType=Capsule, length=0.505, radius=0.505, shapeName=dyn_r_fr_toe_MSTR_capsule, rotation=<rot 0 -90 0>}]
 dyn_top_wireGrp0 [{
   shapeType=Capsule
   length=6.4299
   radius=2.6724
   shapeName=dyn_top_wireGrp0_capsule
   center=<4.7846 0.1098 1.2122>
   rotation=<rot 0 91.194 -70.3565>
  }]
 dyn_top_wireGrp0_0 [{
   shapeType=Capsule
   length=2.7192
   radius=0.66
   shapeName=dyn_top_wireGrp0_0_capsule
   center=<2.4861 0.0311 -0.5914>
   rotation=<rot 0 -88.5483 -80.2317>
  }]
 dyn_top_wireGrp0_1 [{
   shapeType=Capsule
   length=3.6633
   radius=1.0138
   shapeName=dyn_top_wireGrp0_1_capsule
   center=<3.2923 -0.2756 1.5078>
   rotation=<rot 0 79.6069 -71.9736>
  }]
 dyn_top_wireGrp_MSTR [{shapeType=Capsule, length=0.505, radius=0.505, shapeName=dyn_top_wireGrp_MSTR_capsule, rotation=<rot 0 -90 0>}]
}
constraints [
 {bone1Name=dyn_l_fr_toe, bone2Name=dyn_l_fr_toe_MSTR, swing1LimitDegrees=0, swing2LimitDegrees=15, twistLimitDegrees=0}
 {bone1Name=dyn_r_fr_toe, bone2Name=dyn_r_fr_toe_MSTR, swing1LimitDegrees=0, swing2LimitDegrees=15, twistLimitDegrees=0}
 {bone1Name=dyn_c_fr_toe, bone2Name=dyn_c_fr_toe_MSTR, swing1LimitDegrees=0, swing2LimitDegrees=15, twistLimitDegrees=0}
 {bone1Name=dyn_top_wireGrp0_1, bone2Name=dyn_top_wireGrp0_0, swing1LimitDegrees=5, swing2LimitDegrees=10, twistLimitDegrees=0}
 {bone1Name=dyn_bot_wireGrp0_0, bone2Name=dyn_bot_wireGrp0, swing1LimitDegrees=30, swing2LimitDegrees=65, twistLimitDegrees=0}
 {bone1Name=dyn_bot_wireGrp0_2, bone2Name=dyn_bot_wireGrp0, swing1LimitDegrees=40, swing2LimitDegrees=65, twistLimitDegrees=0}
 {bone1Name=dyn_bot_wireGrp0_3, bone2Name=dyn_bot_wireGrp0, swing1LimitDegrees=40, swing2LimitDegrees=65, twistLimitDegrees=0}
 {bone1Name=dyn_bot_wireGrp0, bone2Name=dyn_bot_wireGrp_MSTR, swing1LimitDegrees=15, swing2LimitDegrees=20, twistLimitDegrees=0}
 {bone1Name=dyn_c_ba_toe, bone2Name=dyn_c_ba_toe_MSTR, swing1LimitDegrees=0, swing2LimitDegrees=20, twistLimitDegrees=0}
 {bone1Name=dyn_top_wireGrp0_0, bone2Name=dyn_top_wireGrp0, swing1LimitDegrees=5, swing2LimitDegrees=10, twistLimitDegrees=0}
 {bone1Name=dyn_top_wireGrp0, bone2Name=dyn_top_wireGrp_MSTR, swing1LimitDegrees=5, swing2LimitDegrees=10, twistLimitDegrees=0}
 {bone1Name=dyn_bot_wireGrp0_1, bone2Name=dyn_bot_wireGrp0, swing1LimitDegrees=25, swing2LimitDegrees=45, twistLimitDegrees=0}
]
